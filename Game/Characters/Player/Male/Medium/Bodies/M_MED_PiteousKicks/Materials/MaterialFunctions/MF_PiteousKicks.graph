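# Material function for the PiteousKicks character skin that turns super-levelled areas gold, with a timed emissive transition.
# Super-levelling mask and cloth fuzz functions feed a blend toward texture-driven gold attributes, gated by the current level parameter.
# A transition colour is added to emissive within a mask, fading out over a duration measured from a Blueprint-supplied game time.
# Several roughness, specular, SRM override and gold colour nodes do not reach the output.
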MaterialFunction /Game/Characters/Player/Male/Medium/Bodies/M_MED_PiteousKicks/Materials/MaterialFunctions/MF_PiteousKicks
graph {nodes [
  Constant() {r=1}
  Constant() {r=1}
  n1 = TextureSampleParameter2D() {
   group=PiteousKicks_GoldMaterial
   parameterName=GoldTexture_Diffuse
   samplerSource=SSM_Wrap_WorldGroupSettings
   texture=/Game/Characters/Player/Male/Medium/Bodies/M_MED_PiteousKicks_Pine/Skins/Gleam/Textures/T_PiteousKicks_PineGleam_Body_D
  }
  n2 = ScalarParameter() {defaultValue=1, group=PiteousKicks_GoldMaterial, parameterName=BaseColorMultiplier}
  n3 = TextureSampleParameter2D() {
   group=PiteousKicks_GoldMaterial
   parameterName=GoldTexture_SRM
   samplerSource=SSM_Wrap_WorldGroupSettings
   samplerType=SAMPLERTYPE_Masks
   texture=/Game/Characters/Player/Male/Medium/Bodies/M_MED_PiteousKicks_Pine/Skins/Gleam/Textures/T_PiteousKicks_PineGleam_Body_S
  }
  n4 = ScalarParameter() {group=PiteousKicks_GoldMaterial, parameterName=MetalnessBias}
  n5 = TextureSampleParameter2D() {
   group=00_Texture Inputs
   parameterName=Mask
   samplerSource=SSM_Wrap_WorldGroupSettings
   samplerType=SAMPLERTYPE_Masks
   sortPriority=3
   texture=&T_BlackMask
  }
  n6 = PreSkinnedNormal()
  n7 = PreSkinnedPosition()
  n8 = MaterialFunctionCall!(
   Local Normal=ShaderStageSwitch(PixelShader=VertexInterpolator(VS=@n6).PS, VertexShader=@n6)
   Local Position=ShaderStageSwitch(PixelShader=VertexInterpolator(VS=@n7).PS, VertexShader=@n7)
   Material=FunctionInput() {inputName=MaterialAttributes, inputType=FunctionInput_MaterialAttributes, sortPriority=10}
  ) {materialFunction=&/MF_SuperLevellingMask_PiteousKicks}
  n9 = ComponentMask(None=@n8.SuperLevelling_Mask) {r=true}
  n10 = MaterialFunctionCall!(
   Material=@n8.Result
   Ao/Gmap/Skin=@n5.RGB
   Alpha=FunctionInput() {bUsePreviewValueAsDefault=true, inputName=Tech Art Mask, inputType=FunctionInput_Scalar}
   SuperLevelMask=NamedRerouteDeclaration(None=Saturate(None=@n9)) {name=SupeLevel_Mask, nodeColor=<rgba 0 0.0422 1 1>}
  ) {materialFunction=&/MF_ClothFuzz_PiteousKicks}
  n11 = SetMaterialAttributes!(
   MaterialAttributes=@n10.Result
   BaseColor=Saturate(None=Multiply(A=@n1.RGB, B=@n2))
   Specular=@n3.R
   Roughness=@n3.R
   Metallic=Saturate(None=Add(A=@n3.B, B=@n4))
   EmissiveColor=Constant()
   Normal=FunctionInput() {inputName=OriginalNormal, sortPriority=1000}
   Opacity=Constant() {r=1}
  ) {attributeSetTypes [
    #1
    #2
    #3
    #4
    #5
    #6
    #7
   ]}
  n12 = NamedRerouteUsage()
  n13 = Constant()
  n14 = If(
   A=ScalarParameter() {defaultValue=100, group=SuperLevelling_Mask, parameterName=SL_CurrentLevel, sortPriority=3}
   "A > B"=Constant() {r=1}
   "A == B"=@n13
   "A < B"=@n13
  )
  n15 = GetMaterialAttributes(None=BlendMaterialAttributes(A=@n10.Result, B=@n11, Alpha=Multiply(A=@n12, B=@n14))) {attributeGetTypes [#5]}
  n16 = NamedRerouteUsage()
  n17 = VectorParameter() {group=PiteousKicks_GoldMaterial, parameterName=Transition_EmissiveColor, sortPriority=101, defaultValue=<rgba 2 0 1.4795 1>}
  n18 = Subtract(A=Time(), B=ScalarParameter() {defaultValue=-99999, group=SuperLevelling_Mask, parameterName=SL_GameTimeFromBP, sortPriority=100})
  n19 = Add(A=ScalarParameter() {defaultValue=2, group=SuperLevelling_Mask, parameterName=SL_Animation_Duration, sortPriority=101}) {constB=0.1}
  n20 = OneMinus(None=Saturate(None=Divide(A=@n18, B=@n19)))
  n21 = Add(A=@n16, B=Multiply(A=@n17.RGB, B=SmoothStep(Value=@n20) {constMax=0.2}))
  n22 = NamedRerouteUsage()
  FunctionOutput(None=SetMaterialAttributes!(MaterialAttributes=@n15.MaterialAttributes, EmissiveColor=LinearInterpolate(A=@n16, B=@n21, Alpha=@n22)) {attributeSetTypes [#5]})
  n23 = NamedRerouteUsage()
  n24 = ScalarParameter() {defaultValue=1, group=PiteousKicks_GoldMaterial, parameterName=RoughnessMultiplier}
  LinearInterpolate(A=Saturate(None=Multiply(A=@n23, B=@n24)), B=@n23, Alpha=NamedRerouteUsage())
  n25 = NamedRerouteUsage()
  n26 = ScalarParameter() {group=PiteousKicks_GoldMaterial, parameterName=SpecularOffset}
  LinearInterpolate(A=Saturate(None=Add(A=@n25, B=@n26)), B=@n25, Alpha=NamedRerouteUsage())
  n27 = TextureSampleParameter2D() {
   group=PiteousKicks_GoldMaterial
   parameterName=Gold_SRM
   samplerSource=SSM_Wrap_WorldGroupSettings
   samplerType=SAMPLERTYPE_Masks
   sortPriority=201
   texture=/Game/Characters/Player/Male/Medium/Bodies/M_MED_PiteousKicks/Textures/T_PiteousKicks_Body_S
  }
  n28 = GetMaterialAttributes() {attributeGetTypes [#2, #3, #4]}
  n29 = StaticSwitchParameter(
   True=@n27.RGB
   False=MaterialFunctionCall(R=@n28.Specular, G=@n28.Specular, B=@n28.Specular) {materialFunction=&AppendMany}.RGB
  ) {group=PiteousKicks_GoldMaterial, parameterName=OverrideBaseSRM, sortPriority=200}
  NamedRerouteDeclaration(None=ComponentMask(None=@n29) {g=true}) {name=OriginalRoughness, nodeColor=<rgba 0.9732 1 0.9263 1>}
  NamedRerouteDeclaration(None=ComponentMask(None=@n29) {b=true}) {name=OriginalMetallic, nodeColor=<rgba 0.8047 0 1 1>}
  NamedRerouteDeclaration(None=ComponentMask(None=@n29) {r=true}) {name=OriginalSpecular, nodeColor=<rgba 0.2656 0 1 1>}
  NamedRerouteDeclaration(None=Saturate(None=ComponentMask(None=@n8.SuperLevelling_Mask) {g=true})) {name=SuperLevel_Mask_Emissive, nodeColor=<rgba 0 0.1562 1 1>}
  NamedRerouteDeclaration(None=@n15.EmissiveColor) {name=CurrentEmissive, nodeColor=<rgba 0.9922 1 0 1>}
  Reroute(None=@n28.MaterialAttributes)
  VectorParameter() {group=PiteousKicks_GoldMaterial, parameterName=GoldColor, sortPriority=30, defaultValue=<rgba 0.7084 0.4233 0.1022 1>}
 ]}
