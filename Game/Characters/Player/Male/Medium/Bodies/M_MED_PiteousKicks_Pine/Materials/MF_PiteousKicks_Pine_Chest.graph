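# Material function for the PiteousKicks Pine chest that optionally fakes a see-through ribcage on a translucent material.
# A noise-perturbed sphere mask on pre-skinned position defines the area.
# Deferred shading blends base color and emissive, adding a backface/Fresnel emissive tint, while forward and mobile only set opacity.
# A Custom HLSL node blurs a scene texture with eight-direction, quality-scaled taps, using camera-distance-based, temporally dithered blur width.
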MaterialFunction /Game/Characters/Player/Male/Medium/Bodies/M_MED_PiteousKicks_Pine/Materials/MF_PiteousKicks_Pine_Chest
graph {nodes [
  n1 = NamedRerouteUsage()
  n2 = GetMaterialAttributes(None=NamedRerouteUsage()) {attributeGetTypes [#1, #2]}
  n3 = NamedRerouteUsage()
  n4 = NamedRerouteUsage()
  n5 = VectorParameter() {group=PiteousKicksPineChest, parameterName=BacksideEmissive, defaultValue=<rgba 0.1595 0.3906 0.1343 1>}
  n6 = TwoSidedSign()
  n7 = Add(A=OneMinus(None=Saturate(None=@n6)), B=Fresnel() {baseReflectFraction=0})
  n8 = SetMaterialAttributes!(
   MaterialAttributes=@n1
   BaseColor=LinearInterpolate(A=@n2.BaseColor, B=@n3, Alpha=NamedRerouteUsage())
   EmissiveColor=LinearInterpolate(A=LinearInterpolate(A=@n2.EmissiveColor, B=@n3, Alpha=@n4), B=@n5.RGB, Alpha=Saturate(None=@n7))
   Opacity=Constant() {r=1}
  ) {attributeSetTypes [#1, #2, #3]}
  n9 = SetMaterialAttributes!(MaterialAttributes=@n1, Opacity=NamedRerouteUsage()) {attributeSetTypes [#3]}
  n10 = ShadingPathSwitch(Default=@n8, Deferred=@n8, Forward=@n9, Mobile=@n9)
  n11 = NamedRerouteUsage()
  n12 = StaticBoolParameter() {group=PiteousKicksPineChest, parameterName="UseTransparentRibcage (Requires Translucent Blend Mode)", sortPriority=1}
  FunctionOutput(None=NamedRerouteDeclaration(None=StaticSwitch(True=@n10, False=@n11, Value=@n12)) {name=MaterialOut, nodeColor=<rgba 0.7812 1 0 1>})
  n13 = Subtract(
   A=Length(None=WorldPosition() {worldPositionShaderOffset=WPT_CameraRelative}.XYZ) {is=class MaterialExpressionLength}
   B=ScalarParameter() {group=PiteousKicksPineChest, parameterName=BlurDistanceBias}
  )
  n14 = ScalarParameter() {defaultValue=100, group=PiteousKicksPineChest, parameterName=BlurDistanceFade}
  n15 = LinearInterpolate(
   A=QualitySwitch(
    Default=ScalarParameter() {defaultValue=0.005, group=PiteousKicksPineChest, parameterName=BlurNearDefault}
    Epic=ScalarParameter() {defaultValue=0.008, group=PiteousKicksPineChest, parameterName=BlurNearEpic}
   )
   B=ScalarParameter() {defaultValue=0.001, group=PiteousKicksPineChest, parameterName=BlurFar}
   Alpha=Saturate(None=Divide(A=@n13, B=@n14))
  )
  n16 = Multiply(A=@n15, B=ScalarParameter() {defaultValue=2, group=PiteousKicksPineChest, parameterName=BlurDitherMulitplier}) {constB=2}
  n17 = MaterialFunctionCall() {materialFunction=&DitherTemporalAA}
  NamedRerouteDeclaration(None=LinearInterpolate(
    A=@n15
    B=LinearInterpolate(A=@n15, B=@n16, Alpha=@n17.Result)
    Alpha=ScalarParameter() {defaultValue=1, group=PiteousKicksPineChest, parameterName=BlurUseDither}
   )) {name=BlurDistance, nodeColor=<rgba 0 1 0.9766 1>}
  n18 = NamedRerouteUsage()
  n19 = ScalarParameter() {defaultValue=10, parameterName=NoiseStrength}
  n20 = SphereMask(
   A=VectorParameter() {group=PiteousKicksPineChest, parameterName=RibCageSphereMaskOffset, defaultValue=<rgba 0 20 127 1>}.RGB
   B=VertexInterpolator(VS=PreSkinnedPosition()).PS
   Radius=Add(A=ScalarParameter() {defaultValue=14, group=PiteousKicksPineChest, parameterName=RibcageSphereMaskRadius}, B=Multiply(A=@n18, B=@n19))
   Hardness=ScalarParameter() {defaultValue=0.7, group=PiteousKicksPineChest, parameterName=RibcageSphereMaskHardnessHardness}
  )
  NamedRerouteDeclaration(None=Multiply(A=@n20, B=ScalarParameter() {defaultValue=0.6, group=PiteousKicksPineChest, parameterName=RibcageMaskStrengthEmissive})) {name=RibcageSphereMaskEmissive, nodeColor=<rgba 1 0.3984 0 1>}
  n21 = ScalarParameter() {defaultValue=0.65, group=PiteousKicksPineChest, parameterName=RibcageMaskStrengthMobile}
  NamedRerouteDeclaration(None=OneMinus(None=Multiply(A=@n20, B=@n21))) {name=RibcageSphereMaskMobile, nodeColor=<rgba 1 0.3984 0 1>}
  NamedRerouteDeclaration(None=FunctionInput() {inputName=Material, inputType=FunctionInput_MaterialAttributes}) {name=IN_Material, nodeColor=<rgba 1 0 0.3281 1>}
  NamedRerouteDeclaration(None=FunctionInput() {inputName=Noise, inputType=FunctionInput_Scalar, sortPriority=1}) {name=Noise, nodeColor=<rgba 0 0.2031 1 1>}
  MaterialExpressionNamedRerouteUsage_2 = NamedRerouteUsage()
  MaterialExpressionQualitySwitch_0 = QualitySwitch(Default=Constant() {r=1}, Epic=Constant() {r=2})
  n22 = Custom!(stepSize=@MaterialExpressionNamedRerouteUsage_2, iterations=@MaterialExpressionQualitySwitch_0) {
   code="float dist = stepSize / iterations;\r\nint BufferIndex = 2;\r\nfloat2 offsets[8] = { float2(1,0), float2(-1,0), float2(0,1), float2(0,-1),float2(.5,.5), float2(-.5,-.5), float2(-.5,.5), float2(.5,-.5)};\r\nfloat2 UV = GetDefaultSceneTextureUV(Parameters, int(BufferIndex));\r\nfloat2 NewUV = UV;\r\nfloat3 CurColor=0;\r\n\r\nfor (int i = 0; i < (int)iterations; i++)\r\n{\r\n\tfor (int j = 0; j < 8; j++)\r\n\t{\r\n\t\tNewUV = UV + offsets[j] * (i+1) * dist;\r\n\t\tCurColor += SceneTextureLookupCommon(Parameters, NewUV, BufferIndex, false);\r\n\t}\r\n}\r\n\r\nCurColor /= (iterations * 8.0);\r\nreturn CurColor;\r\n"
   showCode=true
   inputs [
    {
     inputName=stepSize
     input {expression=&self:MaterialExpressionNamedRerouteUsage_2}
    }
    {
     inputName=iterations
     input {expression=&self:MaterialExpressionQualitySwitch_0}
    }
   ]
  }
  n23 = Multiply(B=SceneTexture() {sceneTextureId=PPI_SceneDepth}.Color)
  NamedRerouteDeclaration(None=Add(A=@n22, B=ComponentMask(None=@n23) {b=true, g=true, r=true}) {constB=0}) {name=RibSample, nodeColor=<rgba 0 1 0.2734 1>}
  NamedRerouteDeclaration(None=Multiply(A=@n20, B=ScalarParameter() {defaultValue=1, group=PiteousKicksPineChest, parameterName=RibcageMaskStrengthDiffuse})) {name=RibcageSphereMaskDiffuse, nodeColor=<rgba 1 0.3984 0 1>}
 ]}
