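# Parent material for the PiteousKicks Pine male player character body, built from stacked material functions on Material Attributes.
# The shared 2023 character PreFX function feeds a ghostly glow function, then a Pine-specific chest function.
# The base PiteousKicks function takes that result, plus the glow's original normal and PreFX's tech art mask, before the shared PostFX function finishes it.
# It is two-sided and takes its shading model from the expression.
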Material /Game/Characters/Player/Male/Medium/Bodies/M_MED_PiteousKicks_Pine/Materials/M_PiteousKicks_Pine_PARENT
graph {
 properties {
  bUseMaterialAttributes=true
  bUsedWithClothing=true
  bUsedWithMeshDeformer=true
  bUsedWithMorphTargets=true
  bUsedWithNanite=true
  bUsedWithNiagaraMeshParticles=true
  bUsedWithSkeletalMesh=true
  pixelDepthOffsetMode=PDOM_Legacy
  shadingModel=MSM_FromMaterialExpression
  translucencyLightingMode=TLM_SurfacePerPixelLighting
  twoSided=true
  usedShadingModels=Default Lit | Subsurface | Clear Coat | Subsurface Profile | Cloth
 }
 nodes [
  n1 = MaterialFunctionCall() {materialFunction=&MF_CharacterParent_2023_PreFX}
  n2 = MaterialFunctionCall!(MaterialAttributes=@n1.Result, Fresnel=@n1.Result) {materialFunction=&/MF_PiteousKicks_Pine_GhostlyGlow}
  MaterialFunctionCall(
   Material=MaterialFunctionCall!(
    Tech Art Mask=@n1.Tech Art Mask
    MaterialAttributes=MaterialFunctionCall!(Material=@n2.Result, Noise=@n2.Result) {materialFunction=&/MF_PiteousKicks_Pine_Chest}.Result
    OriginalNormal=@n2.OriginalNormal
   ) {materialFunction=/Game/Characters/Player/Male/Medium/Bodies/M_MED_PiteousKicks/Materials/MaterialFunctions/MF_PiteousKicks}.Result
   M=@n1.M
   Pre Skinned Local Position=@n1.M
   Distance Blend=@n1.M
   Fresnel=@n1.M
  ) {materialFunction=&MF_CharacterParent_2023_PostFX}
 ]
}
