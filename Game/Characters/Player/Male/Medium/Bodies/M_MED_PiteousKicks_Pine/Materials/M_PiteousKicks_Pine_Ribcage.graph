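# Default Lit skeletal mesh material for the ribcage of the Pine variant of the PiteousKicks player body, shading the ribs between two tints.
# The blend adds Fresnel to a fake point-light term: the local-space normal map dotted with the direction from pre-skinned position to an Offset parameter.
# Forward and mobile paths swap in brighter color parameters.
# The shared body mask's green channel drives both roughness and specular.
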Material /Game/Characters/Player/Male/Medium/Bodies/M_MED_PiteousKicks_Pine/Materials/M_PiteousKicks_Pine_Ribcage
graph {
 properties {
  bUseMaterialAttributes=true
  bUsedWithMeshDeformer=true
  bUsedWithMorphTargets=true
  bUsedWithSkeletalMesh=true
  pixelDepthOffsetMode=PDOM_Legacy
  usedShadingModels=Default Lit
 }
 nodes [
  n1 = VectorParameter() {parameterName=Offset, defaultValue=<rgba 0 0 120 1>}
  Add(A=ActorPositionWS(), B=@n1.RGB)
  n2 = ScalarParameter() {defaultValue=3, parameterName=ExponentIn}
  n3 = PreSkinnedPosition()
  n4 = Normalize(VectorInput=Subtract(A=@n1.RGB, B=VertexInterpolator(VS=@n3).PS))
  n5 = TextureSample() {samplerType=SAMPLERTYPE_Normal, texture=/Game/Characters/Player/Male/Medium/Bodies/M_MED_PiteousKicks/Textures/T_PiteousKicks_Body_N}
  n6 = Transform(None=@n5.RGB) {transformType=TRANSFORM_Local}
  n7 = Saturate(None=Add(A=Fresnel(ExponentIn=@n2) {baseReflectFraction=0}, B=DotProduct(A=@n4, B=@n6)))
  n8 = LinearInterpolate(
   A=VectorParameter() {parameterName=RibsColor1, defaultValue=<rgba 0 0 0 1>}.RGB
   B=VectorParameter() {parameterName=RibsColor2, defaultValue=<rgba 0.1546 0.401 0.1704 1>}.RGB
   Alpha=@n7
  )
  n9 = LinearInterpolate(
   A=VectorParameter() {parameterName=MobileRibsColor1, defaultValue=<rgba 0.019 0.0729 0.0236 1>}.RGB
   B=VectorParameter() {parameterName=MobileRibsColor2, defaultValue=<rgba 0.265 0.6875 0.2921 1>}.RGB
   Alpha=@n7
  )
  n10 = TextureSample() {samplerType=SAMPLERTYPE_Masks, texture=/Game/Characters/Player/Male/Medium/Bodies/M_MED_PiteousKicks/Textures/T_PiteousKicks_Body_S}
  SetMaterialAttributes!(
   MaterialAttributes
   Base Color=ShadingPathSwitch(Default=@n8, Deferred=@n8, Forward=@n9, Mobile=@n9)
   Normal=@n5.RGB
   Roughness=@n10.G
   Specular=@n10.G
  ) {attributeSetTypes [
    #1
    #2
    #3
    #4
   ]}
  Transform(None=PixelNormalWS())
  WorldPosition()
 ]
}
